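# Looping ambient sound cue for the Battle Pass S26 character token quest Q01, classed under the quest sound class.
# A single looping wave player feeds the cue output with the ambience loop wave.
# A dedicated attenuation asset shapes the spatial falloff, with boosted volume and a concurrency override capping simultaneous instances.
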SoundCue /BattlePassS26_Quests/Sound/Quest_S26_BP_CharacterToken_Q01/IP_Quest_S26_BP_CharacterToken_Q01_Amb_LP_Cue
root=WavePlayer_0
nodes [{
  id=WavePlayer_0
  properties {bLooping=true, soundWaveAssetPtr=&/IP_Quest_S26_BP_CharacterToken_Q01_Amb_LP}
 }]
properties {
 attenuationSettings=&/IP_Quest_S26_BP_CharacterToken_Q01_Atten
 bOverrideConcurrency=true
 maxDistance=1600
 soundClassObject=&Fort_Quests
 volumeMultiplier=1.5
 concurrencyOverrides {maxCount=3}
}
