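# Sound cue for picking up the Season 26 Battle Pass character token quest item (Q01).
# A Random node picks one of three pickup wave variants with equal weight.
# Playback is spatialized through the quest item's own attenuation asset and routed to the non-critical in-game UI sound class.
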SoundCue /BattlePassS26_Quests/Sound/Quest_S26_BP_CharacterToken_Q01/IP_Quest_S26_BP_CharacterToken_Q01_Pickup_Cue
root=Random_0
nodes [
 {id=Random_0, children [WavePlayer_0,WavePlayer_1,WavePlayer_2], properties {weights [1,1,1]}}
 {
  id=WavePlayer_0
  properties {soundWaveAssetPtr=&/IP_Quest_S26_BP_CharacterToken_Q01_Pickup_01}
 }
 {
  id=WavePlayer_1
  properties {soundWaveAssetPtr=&/IP_Quest_S26_BP_CharacterToken_Q01_Pickup_02}
 }
 {
  id=WavePlayer_2
  properties {soundWaveAssetPtr=&/IP_Quest_S26_BP_CharacterToken_Q01_Pickup_03}
 }
]
properties {
 attenuationSettings=&/IP_Quest_S26_BP_CharacterToken_Q01_Atten
 maxDistance=1600
 soundClassObject=&Fort_UI_InGame_NonCritical
 volumeMultiplier=1
}
